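# Physics asset for the CAU_2 hair part of the female medium Elastic character, defining capsule bodies for ponytail, front and side hair.
# Spine, clavicle, neck and head capsules form an upper-body base, joined with free angular motion up to the hair root and ponytail base.
# Angle-limited chains branch into ponytail tips and small left/right strands, with tighter front and side hair.
# An extra constraint cross-links the right small strand to the main ponytail.
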PhysicsAsset /BRCosmetics/Characters/Player/Female/Medium/Bodies/F_MED_Elastic/Meshes/Parts/CAU_Parts/F_MED_Elastic_CAU_2_Physics
bodies {
 cau2_fr_hair_a [{shapeType=Capsule, length=3, radius=1, center=<2.2658 0 -0.0001>, rotation=<rot 90 0 0>}]
 cau2_fr_hair_b [{shapeType=Capsule, length=3.5, radius=1, center=<2.5045 0 0>, rotation=<rot 90 0 0>}]
 cau2_hair [{shapeType=Capsule, length=1, radius=1}]
 cau2_ponytail_a [{shapeType=Capsule, length=2, radius=3, center=<4.067 0.2132 0>, rotation=<rot 90 0 -13.195>}]
 cau2_ponytail_a_base [{shapeType=Capsule, length=1, radius=1}]
 cau2_ponytail_b [{shapeType=Capsule, length=1.5, radius=3.25, center=<3.0902 0.1137 0>, rotation=<rot 90 -3.5763 5.5986>}]
 cau2_ponytail_c [{shapeType=Capsule, length=6, radius=1.5, center=<2.2506 -0.2121 -0.0386>, rotation=<rot -9.0167 4.6102 -29.6235>}]
 cau2_ponytail_small_a_l [{shapeType=Capsule, length=2.4085, radius=0.75, center=<1.7235 -0.0218 0.0247>, rotation=<rot 0 -83.7238 -87.9242>}]
 cau2_ponytail_small_a_r [{shapeType=Capsule, length=3.8104, radius=0.8959, center=<2.5693 0.0259 -0.2663>, rotation=<rot -0.359 -86.9298 -83.8699>}]
 cau2_ponytail_small_b_l [{shapeType=Capsule, length=1.7094, radius=0.8909, center=<1.6931 0.1024 0.2222>, rotation=<rot 0.3944 87.4064 -81.5075>}]
 cau2_ponytail_small_b_r [{shapeType=Capsule, length=2.9716, radius=1.0328, center=<2.3328 0.0679 -0.0543>, rotation=<rot 0 89.2332 -89.1413>}]
 cau2_ponytail_small_c_l [{shapeType=Capsule, length=3.3847, radius=1, center=<2.5914 -0.081 0.4678>, rotation=<rot 0 -95.3884 -95.0434>}]
 cau2_ponytail_small_c_r [{shapeType=Capsule, length=3.5589, radius=1, center=<2.7262 0.0232 -0.1362>, rotation=<rot 0 -88.806 -84.4264>}]
 cau2_ponytail_tip_a [{shapeType=Capsule, length=3.25, radius=0.75, center=<2.4318 -0.0001 0>, rotation=<rot 90 0 0>}]
 cau2_ponytail_tip_b [{shapeType=Capsule, length=3.75, radius=1, center=<2.6471 0 -0.0001>, rotation=<rot 90 0 0>}]
 cau2_ponytail_tip_c [{shapeType=Capsule, length=6.5, radius=1.25, center=<3.8343 0 0>, rotation=<rot 90 0 0>}]
 cau2_ponytail_tip_d [{shapeType=Capsule, length=3.5, radius=1, center=<2.1037 0 -0.0001>, rotation=<rot 90 0 0>}]
 cau2_side_hair_a [{shapeType=Capsule, length=2.25, radius=1, center=<1.4543 -0.8141 0>, rotation=<rot 90 0 29.2422>}]
 cau2_side_hair_b [{shapeType=Capsule, length=2, radius=1, center=<0 -1.2631 0>, rotation=<rot 0 0 90>}]
 clavicle_l [{shapeType=Capsule, length=8.25, radius=6.25, center=<5.4223 -0.5381 -0.2014>, rotation=<rot 84.6968 0 0>}]
 clavicle_r [{shapeType=Capsule, length=8.25, radius=6.25, center=<-5.4223 0.5381 0.2014>, rotation=<rot 84.6968 0 -180>}]
 head [{shapeType=Capsule, length=4, radius=6.5, center=<2.1556 2.982 -0.2727>, rotation=<rot 90 0 0>}]
 neck_01 [{shapeType=Capsule, length=6.25, radius=4.75, center=<0.1129 0.3067 0>, rotation=<rot 90 0 -7.2822>}]
 spine_05 [{shapeType=Capsule, length=9.25, radius=8.5, center=<3.8274 0.2045 0>}]
}
constraints [
 {
  bone1Name=head
  bone2Name=neck_01
  swing1LimitDegrees=45
  swing1Motion=Free
  swing2LimitDegrees=45
  swing2Motion=Free
  twistLimitDegrees=45
  twistMotion=Free
 }
 {
  bone1Name=neck_01
  bone2Name=spine_05
  swing1LimitDegrees=45
  swing1Motion=Free
  swing2LimitDegrees=45
  swing2Motion=Free
  twistLimitDegrees=45
  twistMotion=Free
 }
 {
  bone1Name=clavicle_l
  bone2Name=spine_05
  swing1LimitDegrees=45
  swing1Motion=Free
  swing2LimitDegrees=45
  swing2Motion=Free
  twistLimitDegrees=45
  twistMotion=Free
 }
 {
  bone1Name=clavicle_r
  bone2Name=spine_05
  swing1LimitDegrees=45
  swing1Motion=Free
  swing2LimitDegrees=45
  swing2Motion=Free
  twistLimitDegrees=45
  twistMotion=Free
 }
 {
  bone1Name=cau2_ponytail_a_base
  bone2Name=cau2_hair
  swing1LimitDegrees=45
  swing1Motion=Free
  swing2LimitDegrees=45
  swing2Motion=Free
  twistLimitDegrees=45
  twistMotion=Free
 }
 {
  bone1Name=cau2_hair
  bone2Name=head
  swing1LimitDegrees=45
  swing1Motion=Free
  swing2LimitDegrees=45
  swing2Motion=Free
  twistLimitDegrees=45
  twistMotion=Free
 }
 {bone1Name=cau2_ponytail_a, bone2Name=cau2_ponytail_a_base, swing1LimitDegrees=45, swing2LimitDegrees=45, twistLimitDegrees=5}
 {bone1Name=cau2_ponytail_b, bone2Name=cau2_ponytail_a, swing1LimitDegrees=45, swing2LimitDegrees=45, twistLimitDegrees=5}
 {bone1Name=cau2_ponytail_c, bone2Name=cau2_ponytail_b, swing1LimitDegrees=45, swing2LimitDegrees=45, twistLimitDegrees=5}
 {bone1Name=cau2_ponytail_tip_a, bone2Name=cau2_ponytail_c, swing1LimitDegrees=45, swing2LimitDegrees=45, twistLimitDegrees=5}
 {bone1Name=cau2_ponytail_tip_b, bone2Name=cau2_ponytail_c, swing1LimitDegrees=45, swing2LimitDegrees=45, twistLimitDegrees=5}
 {bone1Name=cau2_ponytail_tip_c, bone2Name=cau2_ponytail_c, swing1LimitDegrees=45, swing2LimitDegrees=45, twistLimitDegrees=5}
 {bone1Name=cau2_ponytail_tip_d, bone2Name=cau2_ponytail_c, swing1LimitDegrees=45, swing2LimitDegrees=45, twistLimitDegrees=5}
 {bone1Name=cau2_side_hair_b, bone2Name=cau2_hair, swing1LimitDegrees=10, swing2LimitDegrees=10, twistLimitDegrees=10}
 {bone1Name=cau2_side_hair_a, bone2Name=cau2_hair, swing1LimitDegrees=10, swing2LimitDegrees=10, twistLimitDegrees=10}
 {bone1Name=cau2_fr_hair_a, bone2Name=cau2_hair, swing1LimitDegrees=5, swing2LimitDegrees=5, twistLimitDegrees=5}
 {bone1Name=cau2_fr_hair_b, bone2Name=cau2_fr_hair_a, swing1LimitDegrees=5, swing2LimitDegrees=5, twistLimitDegrees=5}
 {bone1Name=cau2_ponytail_small_a_l, bone2Name=cau2_ponytail_a, swing1LimitDegrees=15, swing2LimitDegrees=15, twistLimitDegrees=5}
 {bone1Name=cau2_ponytail_small_b_l, bone2Name=cau2_ponytail_small_a_l, swing1LimitDegrees=15, swing2LimitDegrees=15, twistLimitDegrees=5}
 {bone1Name=cau2_ponytail_small_c_l, bone2Name=cau2_ponytail_small_b_l, swing1LimitDegrees=15, swing2LimitDegrees=15, twistLimitDegrees=5}
 {bone1Name=cau2_ponytail_small_a_r, bone2Name=cau2_ponytail_a, swing1LimitDegrees=10, swing2LimitDegrees=10, twistLimitDegrees=5}
 {bone1Name=cau2_ponytail_small_b_r, bone2Name=cau2_ponytail_small_a_r, swing1LimitDegrees=10, swing2LimitDegrees=10, twistLimitDegrees=5}
 {bone1Name=cau2_ponytail_small_c_r, bone2Name=cau2_ponytail_small_b_r, swing1LimitDegrees=15, swing2LimitDegrees=15, twistLimitDegrees=5}
 {bone1Name=cau2_ponytail_small_b_r, bone2Name=cau2_ponytail_b, swing1LimitDegrees=45, swing2LimitDegrees=45, twistLimitDegrees=45}
]
