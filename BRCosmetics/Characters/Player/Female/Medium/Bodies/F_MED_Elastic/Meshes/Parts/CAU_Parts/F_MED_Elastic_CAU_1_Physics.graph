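# Physics asset for the CAU_1 hair part of the F_MED_Elastic female medium player body, defining how its hair bones move.
# Each hair bone has a small capsule body.
# Strands b through i constrain directly to the cau1_hair root with moderate swing and narrow twist limits.
# The hairMain_a bone hangs from the root and hair_a chains from it, both with tight swing limits and locked twist.
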PhysicsAsset /BRCosmetics/Characters/Player/Female/Medium/Bodies/F_MED_Elastic/Meshes/Parts/CAU_Parts/F_MED_Elastic_CAU_1_Physics
bodies {
 cau1_hair [{shapeType=Capsule, length=1, radius=1}]
 cau1_hairMain_a [{shapeType=Capsule, length=1, radius=1, center=<-1 0 0>, rotation=<rot 90 0 0>}]
 cau1_hair_a [{shapeType=Capsule, length=1, radius=1, center=<1.0518 0 0>, rotation=<rot 90 0 0>}]
 cau1_hair_b [{shapeType=Capsule, length=1, radius=1, center=<1.0518 0 0>, rotation=<rot 90 0 0>}]
 cau1_hair_c [{shapeType=Capsule, length=1, radius=1, center=<1.0518 0 0>, rotation=<rot 90 0 0>}]
 cau1_hair_d [{shapeType=Capsule, length=1, radius=1, center=<1.0518 0 0>, rotation=<rot 90 0 0>}]
 cau1_hair_e [{shapeType=Capsule, length=1, radius=1, center=<1.0518 0 0>, rotation=<rot 90 0 0>}]
 cau1_hair_f [{shapeType=Capsule, length=1, radius=1, center=<1.0518 0 0>, rotation=<rot 90 0 0>}]
 cau1_hair_g [{shapeType=Capsule, length=1, radius=1, center=<1.0518 0 0>, rotation=<rot 90 0 0>}]
 cau1_hair_h [{shapeType=Capsule, length=1, radius=1, center=<1.0518 0 0>, rotation=<rot 90 0 0>}]
 cau1_hair_i [{shapeType=Capsule, length=1, radius=1, center=<1.0518 0 0>, rotation=<rot 90 0 0>}]
}
constraints [
 {bone1Name=cau1_hair_c, bone2Name=cau1_hair, swing1LimitDegrees=25, swing2LimitDegrees=25, twistLimitDegrees=5}
 {bone1Name=cau1_hair_b, bone2Name=cau1_hair, swing1LimitDegrees=25, swing2LimitDegrees=25, twistLimitDegrees=5}
 {bone1Name=cau1_hair_d, bone2Name=cau1_hair, swing1LimitDegrees=25, swing2LimitDegrees=25, twistLimitDegrees=5}
 {bone1Name=cau1_hair_f, bone2Name=cau1_hair, swing1LimitDegrees=25, swing2LimitDegrees=25, twistLimitDegrees=5}
 {bone1Name=cau1_hair_h, bone2Name=cau1_hair, swing1LimitDegrees=25, swing2LimitDegrees=25, twistLimitDegrees=5}
 {bone1Name=cau1_hair_i, bone2Name=cau1_hair, swing1LimitDegrees=25, swing2LimitDegrees=25, twistLimitDegrees=5}
 {bone1Name=cau1_hair_g, bone2Name=cau1_hair, swing1LimitDegrees=25, swing2LimitDegrees=25, twistLimitDegrees=5}
 {bone1Name=cau1_hair_e, bone2Name=cau1_hair, swing1LimitDegrees=25, swing2LimitDegrees=25, twistLimitDegrees=5}
 {bone1Name=cau1_hairMain_a, bone2Name=cau1_hair, swing1LimitDegrees=5, swing2LimitDegrees=5, twistLimitDegrees=5, twistMotion=Locked}
 {bone1Name=cau1_hair_a, bone2Name=cau1_hairMain_a, swing1LimitDegrees=5, swing2LimitDegrees=5, twistLimitDegrees=5, twistMotion=Locked}
]
